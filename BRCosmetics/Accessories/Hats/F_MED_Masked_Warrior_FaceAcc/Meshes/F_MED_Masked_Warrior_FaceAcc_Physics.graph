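# Physics asset for the F_MED_Masked_Warrior_FaceAcc cosmetic mesh, defining capsule bodies and joints for its dynamic hair and hood bones.
# A head capsule anchors a three-bone front hair chain, a two-bone right chain and a single left hair bone.
# Angular limits on the hair constrain its swing and twist, with the second right bone held nearly rigid to the first.
# The hood bone's swing and twist are fully locked to the head.
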PhysicsAsset /BRCosmetics/Accessories/Hats/F_MED_Masked_Warrior_FaceAcc/Meshes/F_MED_Masked_Warrior_FaceAcc_Physics
bodies {
 dyn_hair_f_01 [{shapeType=Capsule, length=4.0253, radius=2.2519, center=<2.8593 0 0>, rotation=<rot 90 0 0>}]
 dyn_hair_f_02 [{shapeType=Capsule, length=3.1973, radius=2.3062, center=<1 0 0>, rotation=<rot 90 0 -10.8147>}]
 dyn_hair_f_03 [{shapeType=Capsule, length=2.9048, radius=1.3473, center=<1.4617 0 0>, rotation=<rot 90 0 0>}]
 dyn_hair_l_01 [{shapeType=Capsule, length=3.8627, radius=2.3238, center=<0 0 -1>, rotation=<rot -14.6497 0 0>}]
 dyn_hair_r_01 [{shapeType=Capsule, length=2.8473, radius=1.9184, center=<0 0 -1>}]
 dyn_hair_r_02 [{shapeType=Capsule, length=1.8419, radius=2.0977, center=<0 0 -1>}]
 dyn_hood_01 [{shapeType=Capsule, length=1.091, radius=2.25, center=<0 0 -0.4947>}]
 head [{shapeType=Capsule, length=7.1359, radius=6.4831, center=<0 0.6598 0>, rotation=<rot 90 0 0>}]
}
constraints [
 {bone1Name=dyn_hair_l_01, bone2Name=head, swing1LimitDegrees=45, swing2LimitDegrees=45, twistLimitDegrees=15}
 {bone1Name=dyn_hair_r_01, bone2Name=head, swing1LimitDegrees=45, swing2LimitDegrees=45, twistLimitDegrees=15}
 {bone1Name=dyn_hair_r_02, bone2Name=dyn_hair_r_01, swing1LimitDegrees=1, swing2LimitDegrees=5, twistLimitDegrees=5}
 {bone1Name=dyn_hair_f_01, bone2Name=head, swing1LimitDegrees=30, swing2LimitDegrees=30, twistLimitDegrees=5}
 {bone1Name=dyn_hair_f_02, bone2Name=dyn_hair_f_01, swing1LimitDegrees=30, swing2LimitDegrees=30, twistLimitDegrees=10}
 {bone1Name=dyn_hair_f_03, bone2Name=dyn_hair_f_02, swing1LimitDegrees=25, swing2LimitDegrees=25, twistLimitDegrees=10}
 {
  bone1Name=dyn_hood_01
  bone2Name=head
  swing1LimitDegrees=2
  swing1Motion=Locked
  swing2LimitDegrees=2
  swing2Motion=Locked
  twistLimitDegrees=2
  twistMotion=Locked
 }
]
