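# Post-process style AnimBP for the Masked Warrior face accessory, animating its hair bones on top of the pose copied from the attached parent mesh.
# When not targeting, chained AnimDynamics on three front hair bones feed a RigidBody node; when targeting, arm-lift-driven offsets move side hair instead.
# A constant right-hair offset follows the blend.
# EventGraph derives lift alphas from upperarm-to-clavicle yaw, plus an unused head-based corrective.
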AnimBlueprint /BRCosmetics/Accessories/Hats/F_MED_Masked_Warrior_FaceAcc/Meshes/F_MED_Masked_Warrior_FaceAcc_Skeleton_AnimBP
target_skeleton=&/F_MED_Masked_Warrior_FaceAcc_Skeleton
graphs [
 AnimGraph {nodes [
   n1 = Anim Dynamics - Bone: dyn_hair_f_01(ComponentPose=Local To Component(LocalPose=Use cached pose 'SavedPose22321'()), Alpha=1, ExternalForce="0,0,0") {
    angularSpringConstant=300
    bGravityOverrideInSimSpace=true
    bLinearSpring=true
    bOverrideLinearDamping=true
    bUseGravityOverride=true
    linearDampingOverride=0.99
    simulationSpace=RootRelative
    alphaBoolBlend {blendOption=Linear}
    boundBone {boneName=dyn_hair_f_01}
    componentAppliedLinearAccClamp=<2000 2000 2000>
    componentLinearAccScale=<0.3 0.3 0.3>
    componentLinearVelScale=<0.1 0.1 0.1>
    gravityOverride {z=-980}
    physicsBodyDefinitions [{
      sphereCollisionRadius=0
      boundBone {boneName=dyn_hair_f_01}
      boxExtents=<5 2 5>
      constraintSetup {angularLimitsMax=<5 5 10>, angularLimitsMin=<-5 -5 0>, angularTarget=<1 0 0>}
      localJointOffset=<0.1 0 0>
     }]
    relativeSpaceBone {boneName=root}
   }
   n2 = Anim Dynamics - Bone: dyn_hair_f_02(ComponentPose=@n1, Alpha=1, ExternalForce="0,0,0") {
    angularSpringConstant=300
    bGravityOverrideInSimSpace=true
    bLinearSpring=true
    bOverrideLinearDamping=true
    bUseGravityOverride=true
    linearDampingOverride=0.99
    simulationSpace=RootRelative
    alphaBoolBlend {blendOption=Linear}
    boundBone {boneName=dyn_hair_f_02}
    componentAppliedLinearAccClamp=<2000 2000 2000>
    componentLinearAccScale=<0.3 0.3 0.3>
    componentLinearVelScale=<0.1 0.1 0.1>
    gravityOverride {z=-980}
    physicsBodyDefinitions [{
      sphereCollisionRadius=0
      boundBone {boneName=dyn_hair_f_02}
      boxExtents=<5 2 5>
      constraintSetup {angularLimitsMax=<5 5 5>, angularLimitsMin=<-5 -5 -2>, angularTarget=<1 0 0>}
      localJointOffset=<0.1 0 0>
     }]
    relativeSpaceBone {boneName=root}
   }
   n3 = Anim Dynamics - Bone: dyn_hair_f_03(ComponentPose=@n2, Alpha=1, ExternalForce="0,0,0") {
    angularSpringConstant=300
    bGravityOverrideInSimSpace=true
    bLinearSpring=true
    bOverrideLinearDamping=true
    bUseGravityOverride=true
    linearDampingOverride=0.99
    simulationSpace=RootRelative
    alphaBoolBlend {blendOption=Linear}
    boundBone {boneName=dyn_hair_f_03}
    componentAppliedLinearAccClamp=<2000 2000 2000>
    componentLinearAccScale=<0.3 0.3 0.3>
    componentLinearVelScale=<0.1 0.1 0.1>
    gravityOverride {z=-980}
    physicsBodyDefinitions [{
      sphereCollisionRadius=0
      boundBone {boneName=dyn_hair_f_03}
      boxExtents=<5 2 5>
      constraintSetup {angularLimitsMax=<5 5 5>, angularLimitsMin=<-5 -5 -5>, angularTarget=<1 0 0>}
      localJointOffset=<0.1 0 0>
     }]
    relativeSpaceBone {boneName=root}
   }
   n4 = RigidBody(ComponentPose=@n3) {
    overridePhysicsAsset=&/F_MED_Masked_Warrior_FaceAcc_Physics
    simulationSpace=BaseBoneSpace
    alphaBoolBlend {blendOption=Linear}
    baseBoneRef {boneName=spine_03}
    componentAppliedLinearAccClamp=<2000 2000 2000>
    componentLinearAccScale=<1 1 1>
   }
   n5 = Component To Local(ComponentPose=@n4)
   n6 = "Transform (Modify) Bone - Bone: dyn_hair_r_01"(
    ComponentPose=Local To Component(LocalPose=Use cached pose 'SavedPose22321'())
    Alpha=Get armSideLiftR() -> (armSideLiftR:double)
    Translation="0,-10,0"
    Rotation="0,-45,0"
    Scale="0,0,0"
   ) {
    rotationMode=BMM_Additive
    rotationSpace=BCS_BoneSpace
    scaleSpace=BCS_BoneSpace
    translationMode=BMM_Additive
    translationSpace=BCS_BoneSpace
    alphaBoolBlend {blendOption=Linear}
    boneToModify {boneName=dyn_hair_r_01}
    rotation {yaw=-45}
    scale=<0 0 0>
    translation {y=-10}
   }
   n7 = "Transform (Modify) Bone - Bone: dyn_hair_l_01"(ComponentPose=@n6, Alpha=Get armSideLiftL() -> (armSideLiftL:double), Translation="3,0,0", Rotation="0,45,0", Scale="0,0,0") {
    rotationMode=BMM_Additive
    rotationSpace=BCS_BoneSpace
    scaleSpace=BCS_BoneSpace
    translationMode=BMM_Additive
    translationSpace=BCS_BoneSpace
    alphaBoolBlend {blendOption=Linear}
    boneToModify {boneName=dyn_hair_l_01}
    rotation {yaw=45}
    scale=<0 0 0>
    translation {x=3}
   }
   n8 = Component To Local(ComponentPose=@n7)
   n9 = Get bIsTargeting()
   n10 = "Transform (Modify) Bone - Bone: dyn_hair_r_01"(
    ComponentPose=Local To Component(LocalPose=Blend Poses by bool(BlendPose_0=@n8, BlendPose_1=@n5, bActiveValue=@n9) {blendType=Linear})
    Alpha=1
    Translation="1,0,0"
    Rotation="20,0,0"
    Scale="1,1,1"
   ) {
    rotationMode=BMM_Additive
    rotationSpace=BCS_BoneSpace
    translationMode=BMM_Additive
    translationSpace=BCS_BoneSpace
    alphaBoolBlend {blendOption=Linear}
    boneToModify {boneName=dyn_hair_r_01}
    rotation {pitch=20}
    translation {x=1}
   }
   Output Pose(Result=Component To Local(ComponentPose=@n10)) {layerGroup=None}
   Save cached pose 'SavedPose22321'(Pose=Copy Pose From Mesh() {bUseAttachedParent=true}) {cacheName=SavedPose22321}
   Get hair_r_01_corrective() -> (hair_r_01_corrective:double)
  ]}
 EventGraph {nodes [
   Event BlueprintUpdateAnimation()
   n1 = GetOwnerSkeletalMeshComponent()
   n2 = GetDeltaTransformFromRefPose(self=@n1, BoneName=upperarm_r, BaseName=clavicle_r)
   n3 = "Absolute (Float)"(A=BreakRotator(InRot=BreakTransform(InTransform=@n2).Rotation).Yaw)
   > Set armSideLiftR(armSideLiftR:double=float * float(A=MapRangeClamped(Value=@n3, InRangeA=40, InRangeB=120), B=1.5)) -> (Output_Get:double)
   n4 = GetDeltaTransformFromRefPose(self=@n1, BoneName=upperarm_l, BaseName=clavicle_l)
   n5 = "Absolute (Float)"(A=BreakRotator(InRot=BreakTransform(InTransform=@n4).Rotation).Yaw)
   > Set armSideLiftL(armSideLiftL:double=float * float(A=MapRangeClamped(Value=@n5, InRangeA=15, InRangeB=90), B=0.5)) -> (Output_Get:double)
   n6 = BreakTransform(InTransform=GetDeltaTransformFromRefPose(self=@n1, BoneName=head, BaseName=neck_02))
   > Set hair_r_01_corrective(hair_r_01_corrective:double=MapRangeClamped(Value=BreakRotator(InRot=@n6.Rotation), InRangeA=0, InRangeB=-30)) -> (Output_Get:double)
   TryGetPawnOwner()
  ]}
]
variables [
 armSideLiftL {type=number}
 armSideLiftR {type=number}
 hair_r_01_corrective {type=number}
]
