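# Physics asset for the F_MED Masked Warrior hood face accessory, defining capsule bodies and joints for its dyn_hair bones.
# Capsules on head, upper spine and left clavicle cover the body, with head and clavicle jointed freely to spine_05.
# Hair bodies hang directly off the head, except main_02 and main_03, which branch from main_01.
# The back piece is rotationally locked to the head, and main_03 has its twist locked.
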PhysicsAsset /BRCosmetics/Accessories/Hats/F_MED_Masked_Warrior_FaceAcc/Meshes/F_MED_Masked_Warrior_Hood_FaceAcc_Physics
bodies {
 clavicle_l [{shapeType=Capsule, length=15.7726, radius=4.9547, center=<6.8982 0 1.5012>, rotation=<rot 90 0 0>}]
 dyn_hair_back [{shapeType=Capsule, length=5, radius=2, rotation=<rot 90 0 0>}]
 dyn_hair_br [{shapeType=Capsule, length=4.132, radius=2, center=<0.3685 0 0>, rotation=<rot 90 0 0>}]
 dyn_hair_lb [{shapeType=Capsule, length=4.4229, radius=2, center=<0.6935 0.0001 0>, rotation=<rot 90 0 0>}]
 dyn_hair_lt [{shapeType=Capsule, length=1.7789, radius=2, center=<0.3239 0 0>, rotation=<rot 90 0 0>}]
 dyn_hair_main_01 [{shapeType=Capsule, length=3.7032, radius=2, center=<0.2161 0 0.0001>, rotation=<rot 90 0 0>}]
 dyn_hair_main_02 [{shapeType=Capsule, length=1.7187, radius=1.6921, center=<0.4909 0.2255 0.3085>, rotation=<rot 41.269 -180 -180>}]
 dyn_hair_main_03 [{shapeType=Capsule, length=4.337, radius=2.9695, center=<1.7452 0.7533 1.1071>, rotation=<rot 68.0004 52.641 39.6469>}]
 dyn_hair_tl [{shapeType=Capsule, length=3.6547, radius=2, center=<0.4935 0 0>, rotation=<rot 90 0 0>}]
 head [{shapeType=Capsule, length=4.1186, radius=8.3511, center=<4.6565 -0.6436 -2.4792>, rotation=<rot 90 0 0>}]
 spine_05 [{shapeType=Capsule, length=18.3557, radius=9.869, center=<0 1.8308 0>}]
}
constraints [
 {bone1Name=dyn_hair_main_01, bone2Name=head, swing1LimitDegrees=30, swing2LimitDegrees=30, twistLimitDegrees=10}
 {
  bone1Name=dyn_hair_back
  bone2Name=head
  swing1LimitDegrees=45
  swing1Motion=Locked
  swing2LimitDegrees=45
  swing2Motion=Locked
  twistLimitDegrees=15
  twistMotion=Locked
 }
 {bone1Name=dyn_hair_main_02, bone2Name=dyn_hair_main_01, swing1LimitDegrees=45, swing2LimitDegrees=45, twistLimitDegrees=15}
 {
  bone1Name=dyn_hair_main_03
  bone2Name=dyn_hair_main_01
  swing1LimitDegrees=45
  swing2LimitDegrees=16
  twistLimitDegrees=12
  twistMotion=Locked
 }
 {bone1Name=dyn_hair_br, bone2Name=head, swing1LimitDegrees=45, swing2LimitDegrees=45, twistLimitDegrees=15}
 {bone1Name=dyn_hair_lb, bone2Name=head, swing1LimitDegrees=45, swing2LimitDegrees=45, twistLimitDegrees=15}
 {bone1Name=dyn_hair_lt, bone2Name=head, swing1LimitDegrees=45, swing2LimitDegrees=45, twistLimitDegrees=15}
 {bone1Name=dyn_hair_tl, bone2Name=head, swing1LimitDegrees=45, swing2LimitDegrees=45, twistLimitDegrees=15}
 {
  bone1Name=head
  bone2Name=spine_05
  swing1LimitDegrees=45
  swing1Motion=Free
  swing2LimitDegrees=45
  swing2Motion=Free
  twistLimitDegrees=45
  twistMotion=Free
 }
 {
  bone1Name=clavicle_l
  bone2Name=spine_05
  swing1LimitDegrees=45
  swing1Motion=Free
  swing2LimitDegrees=45
  swing2Motion=Free
  twistLimitDegrees=45
  twistMotion=Free
 }
]
