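# Animation Blueprint for the Masked Warrior hood accessory, adding secondary motion to its hood and hair bones.
# It copies the attached parent's pose, then chains spring AnimDynamics, a left-upperarm Look At, additive bone offsets and a RigidBody simulation.
# Skydiving external force feeds the physics nodes.
# The event graph maps head and upper-arm rotations relative to spine_05 into alphas; only left head-twist and arm-swing values drive the offsets.
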AnimBlueprint /BRCosmetics/Accessories/Hats/F_MED_Masked_Warrior_FaceAcc/Meshes/F_MED_Masked_Warrior_Hood_AnimBP
target_skeleton=&/F_MED_Masked_Warrior_Hood_Skeleton
graphs [
 AnimGraph {nodes [
   n1 = Break Fort Anim Input Skydiving External Force(FortAnimInput_SkydivingExternalForce=Get SkydivingExternalForceInput()) -> (
    bUseSkydivingVectorForce=~
    SkydivingVectorForce=~
    HeadToPelvisDirection=~
    FloatingMultiplier=~
    FloatingAdditive=~
    DivingMultiplier=~
    DivingAdditive=~
    ParachutingMultiplier=~
    ParachutingAdditive=~
   )
   n2 = Anim Dynamics - Bone: dyn_hair_top(ComponentPose=Local To Component(LocalPose=Copy Pose From Mesh() {bUseAttachedParent=true}), Alpha=1, ExternalForce=@n1.SkydivingVectorForce) {
    bGravityOverrideInSimSpace=true
    bLinearSpring=true
    bOverrideLinearDamping=true
    bUseGravityOverride=true
    linearDampingOverride=0.999
    linearSpringConstant=300
    alphaBoolBlend {blendOption=Linear}
    boundBone {boneName=dyn_hair_top}
    componentAppliedLinearAccClamp=<2000 2000 2000>
    componentLinearAccScale=<0.3 0.3 0.3>
    componentLinearVelScale=<0.1 0.1 0.1>
    gravityOverride {z=-980}
    physicsBodyDefinitions [{
      sphereCollisionRadius=0
      boundBone {boneName=dyn_hair_top}
      boxExtents=<2 2 2>
      constraintSetup {angularLimitsMax=<0 0 0>, angularLimitsMin=<0 0 0>, angularTarget=<0 0 0>, linearAxesMax=<0.5 0.5 0.5>, linearAxesMin=<-0.5 -0.5 -0.5>}
      localJointOffset=<0 0 0.05>
     }]
    relativeSpaceBone {boneName=head}
   }
   n3 = Anim Dynamics - Bone: dyn_hair_back(ComponentPose=@n2, Alpha=1, ExternalForce=@n1.SkydivingVectorForce) {
    bGravityOverrideInSimSpace=true
    bLinearSpring=true
    bOverrideLinearDamping=true
    bUseGravityOverride=true
    linearDampingOverride=0.99
    linearSpringConstant=200
    simulationSpace=RootRelative
    alphaBoolBlend {blendOption=Linear}
    boundBone {boneName=dyn_hair_back}
    componentAppliedLinearAccClamp=<2000 2000 2000>
    componentLinearAccScale=<0.3 0.3 0.3>
    componentLinearVelScale=<0.1 0.1 0.1>
    gravityOverride {z=-980}
    physicsBodyDefinitions [{
      sphereCollisionRadius=0
      boundBone {boneName=dyn_hair_back}
      boxExtents=<2 2 2>
      constraintSetup {angularLimitsMax=<0 0 0>, angularLimitsMin=<0 0 0>, angularTarget=<0 0 0>, linearAxesMax=<0.5 1 1>, linearAxesMin=<-3 -1 0>}
      localJointOffset=<0 0 0.05>
     }]
    relativeSpaceBone {boneName=root}
   }
   n4 = Anim Dynamics - Bone: dyn_hood(ComponentPose=@n3, Alpha=1, ExternalForce=@n1.SkydivingVectorForce) {
    bGravityOverrideInSimSpace=true
    bLinearSpring=true
    bOverrideLinearDamping=true
    bUseGravityOverride=true
    linearDampingOverride=0.99
    linearSpringConstant=100
    simulationSpace=RootRelative
    alphaBoolBlend {blendOption=Linear}
    boundBone {boneName=dyn_hood}
    componentAppliedLinearAccClamp=<2000 2000 2000>
    componentLinearAccScale=<0.3 0.3 0.3>
    componentLinearVelScale=<0.1 0.1 0.1>
    gravityOverride {z=-980}
    physicsBodyDefinitions [{
      sphereCollisionRadius=0
      boundBone {boneName=dyn_hood}
      boxExtents=<20 20 10>
      constraintSetup {angularLimitsMax=<0 0 0>, angularLimitsMin=<0 0 0>, angularTarget=<0 0 0>, linearAxesMax=<3 0.75 1>, linearAxesMin=<-3 -0.75 -0.2>}
      localJointOffset=<0 0 0.05>
     }]
    relativeSpaceBone {boneName=root}
   }
   n5 = Look At - Bone: dyn_hair_main_01(ComponentPose=@n4, Alpha=1) {
    lookAtClamp=33
    alphaBoolBlend {blendOption=Linear}
    boneToModify {boneName=dyn_hair_main_01}
    lookAtLocation {x=0, z=5}
    lookAtTarget {boneReference {boneName=upperarm_l}}
    lookAt_Axis {axis=<1 0>}
    lookUp_Axis {axis {x=0, z=1}}
   }
   n6 = "Transform (Modify) Bone - Bone: dyn_hair_main_01"(ComponentPose=@n5, Alpha=Get headTwist_l() -> (headTwist_l:double), Translation="2,0,2", Rotation="0,0,0", Scale="1,1,1") {
    translationMode=BMM_Additive
    translationSpace=BCS_BoneSpace
    alphaBoolBlend {blendOption=Linear}
    boneToModify {boneName=dyn_hair_main_01}
    translation {x=2, z=2}
   }
   n7 = "Transform (Modify) Bone - Bone: dyn_hair_main_01"(ComponentPose=@n6, Alpha=Get armSwingDelta_l() -> (armSwingDelta_l:double), Translation="-2,0,2.285", Rotation="0,0,0", Scale="1,1,1") {
    translationMode=BMM_Additive
    translationSpace=BCS_BoneSpace
    alphaBoolBlend {blendOption=Linear}
    boneToModify {boneName=dyn_hair_main_01}
    translation {x=-2, z=2.285}
   }
   n8 = RigidBody(ComponentPose=@n7, ExternalForce=@n1.SkydivingVectorForce) {
    bOverrideWorldGravity=true
    overridePhysicsAsset=&/F_MED_Masked_Warrior_Hood_FaceAcc_Physics
    simulationSpace=BaseBoneSpace
    alphaBoolBlend {blendOption=Linear}
    baseBoneRef {boneName=root}
    componentLinearAccScale=<0.5 0.5 0.5>
    overrideWorldGravity {z=-980}
   }
   Output Pose(Result=Component To Local(ComponentPose=@n8)) {layerGroup=None}
  ]}
 EventGraph {nodes [
   Event BlueprintUpdateAnimation()
   n1 = GetOwnerSkeletalMeshComponent()
   n2 = BreakTransform(InTransform=GetDeltaTransformFromRefPose(self=@n1, BoneName=head, BaseName=spine_05))
   > Set headSwingDelta(headSwingDelta:double=MapRangeClamped(Value=BreakRotator(InRot=@n2.Rotation).Yaw, InRangeA=17, InRangeB=-44)) -> (Output_Get:double)
   n3 = BreakRotator(InRot=@n2.Rotation)
   > Set headTwist_r(headTwist_r:double=MapRangeClamped(Value=@n3, InRangeA=15, InRangeB=-37)) -> (Output_Get:double)
   > Set headTwist_l(headTwist_l:double=MapRangeClamped(Value=@n3, InRangeA=15, InRangeB=65)) -> (Output_Get:double)
   n4 = BreakTransform(InTransform=GetDeltaTransformFromRefPose(self=@n1, BoneName=upperarm_r, BaseName=spine_05))
   > Set armSwingDelta_r(armSwingDelta_r:double=MapRangeClamped(Value=BreakRotator(InRot=@n4.Rotation).Yaw, InRangeA=45, InRangeB=130)) -> (Output_Get:double)
   n5 = BreakTransform(InTransform=GetDeltaTransformFromRefPose(self=@n1, BoneName=upperarm_l, BaseName=spine_05))
   > Set armSwingDelta_l(armSwingDelta_l:double=MapRangeClamped(Value=BreakRotator(InRot=@n5.Rotation).Yaw, InRangeA=12, InRangeB=100)) -> (Output_Get:double)
   TryGetPawnOwner()
  ]}
]
variables [
 armSwingDelta_l {type=number}
 armSwingDelta_r {type=number}
 headSwingDelta {type=number}
 headTwist_l {type=number}
 headTwist_r {type=number}
]
